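# Looping sound cue for the BRSword pull stress, layering ice stress, energy, ambient, electrical arc and sparks.
# Parallel close and distant mixes use separate wave sets and Carmine attenuations, then blend under a sound class override and slight random pitch.
# Delays stagger ice, one-shot sparks and arc, while envelopes fade in loops and raise energy and arc pitch to build tension.
# A concurrency override prevents new instances at the limit.
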SoundCue /FNGameplayCues/Sounds/Fort_GamePlay_Sounds/BRSword/AB_Pull_Stress_Loop_cue
root=Modulator_2
nodes [
 {id=Attenuation_1, children [Mixer_1], properties {attenuationSettings=/Game/Sounds/Attenuation/Fort_Weapon_Carmine_Close_Athena_Att}}
 {id=Attenuation_2, children [Mixer_0], properties {attenuationSettings=/Game/Sounds/Attenuation/Fort_Weapon_Carmine_Far_Athena_Att}}
 {id=Delay_0, children [WavePlayer_1], properties {delayMax=2, delayMin=2}}
 {id=Delay_1, children [WavePlayer_2], properties {delayMax=2.25, delayMin=2.25}}
 {id=Delay_12, children [WavePlayer_16], properties {delayMax=0.7, delayMin=0.7}}
 {id=Delay_13, children [WavePlayer_19], properties {delayMax=2, delayMin=2}}
 {id=Delay_14, children [WavePlayer_20], properties {delayMax=2.25, delayMin=2.25}}
 {id=Delay_15, children [WavePlayer_21], properties {delayMax=2.75, delayMin=2.75}}
 {id=Delay_16, children [Enveloper_8], properties {delayMax=3.8, delayMin=3.8}}
 {id=Delay_17, children [Modulator_3], properties {delayMax=0.1, delayMin=0.1}}
 {id=Delay_2, children [WavePlayer_5], properties {delayMax=0.7, delayMin=0.7}}
 {id=Delay_3, children [WavePlayer_3], properties {delayMax=2.75, delayMin=2.75}}
 {id=Delay_4, children [Enveloper_1], properties {delayMax=3.8, delayMin=3.8}}
 {id=Delay_5, children [Modulator_0], properties {delayMax=0.1, delayMin=0.1}}
 {
  id=Enveloper_0
  children [WavePlayer_0]
  properties {
   pitchCurve {editorCurveData {keys [{time=0}]}}
   volumeCurve {editorCurveData {keys [<at 0.0607 0.0571>,<at 0.5985 0.6139>,<at 5.7725 1.2463>,<at 12.0495 1.2592>]}}
  }
 }
 {
  id=Enveloper_1
  children [WavePlayer_7]
  properties {
   pitchCurve {editorCurveData {keys [{time=0}, <at 7.9941 1.3642>]}}
   volumeCurve {editorCurveData {keys [<at 0.0019 0.0116>,<at 0.1817 0.999>]}}
  }
 }
 {
  id=Enveloper_2
  children [WavePlayer_6]
  properties {
   pitchCurve {editorCurveData {keys [{time=0}, <at 7.9941 1.3642>]}}
   volumeCurve {editorCurveData {keys [<at 0.0019 0.0116>,<at 2.5304 0.9977>,<at 12.0495 1.2592>]}}
  }
 }
 {
  id=Enveloper_6
  children [WavePlayer_17]
  properties {
   pitchCurve {editorCurveData {keys [{time=0}, <at 7.9941 1.3642>]}}
   volumeCurve {editorCurveData {keys [<at 0.0019 0.0116>,<at 2.5304 0.9977>,<at 12.0495 1.2592>]}}
  }
 }
 {
  id=Enveloper_7
  children [WavePlayer_18]
  properties {
   pitchCurve {editorCurveData {keys [{time=0}]}}
   volumeCurve {editorCurveData {keys [<at 0.0607 0.0571>,<at 0.5985 0.6139>,<at 5.7725 1.2463>,<at 12.0495 1.2592>]}}
  }
 }
 {
  id=Enveloper_8
  children [WavePlayer_22]
  properties {
   pitchCurve {editorCurveData {keys [{time=0}, <at 7.9941 1.3642>]}}
   volumeCurve {editorCurveData {keys [<at 0.0019 0.0116>,<at 0.1817 0.999>]}}
  }
 }
 {
  id=Mixer_0
  children [Delay_12,Enveloper_6,Enveloper_7,Delay_13,Delay_14,Delay_15,Delay_16,Delay_17]
  properties {inputVolume [0.5,1.5,1,0.5,0.5,0.5,1.5,0.45]}
 }
 {
  id=Mixer_1
  children [Delay_2,Enveloper_2,Enveloper_0,Delay_0,Delay_1,Delay_3,Delay_4,Delay_5]
  properties {inputVolume [1,1.5,1,0.5,0.5,0.5,1.5,0.45]}
 }
 {id=Mixer_2, children [SoundClass_1,SoundClass_2], properties {inputVolume [0.9,1]}}
 {id=Modulator_0, children [WavePlayer_4], properties {pitchMax=0.7, pitchMin=0.7}}
 {id=Modulator_2, children [Mixer_2], properties {pitchMax=1.1, pitchMin=1}}
 {id=Modulator_3, children [WavePlayer_23], properties {pitchMax=0.7, pitchMin=0.7}}
 {
  id=SoundClass_1
  children [Attenuation_1]
  properties {soundClassOverride=&Fort_WeaponRanged_Fire_3P_PSM}
 }
 {
  id=SoundClass_2
  children [Attenuation_2]
  properties {soundClassOverride=&Fort_WeaponRanged_Fire_3P_PSM}
 }
 {id=WavePlayer_0, properties {bLooping=true, soundWaveAssetPtr=&/AB_Ambient_Loop_01}}
 {id=WavePlayer_1, properties {soundWaveAssetPtr=&/AB_Pull_Loop_Sparks_01}}
 {
  id=WavePlayer_16
  properties {bLooping=true, soundWaveAssetPtr=&/AB_Pull_Ice_Stress_Loop_01_distant}
 }
 {
  id=WavePlayer_17
  properties {bLooping=true, soundWaveAssetPtr=&/AB_Pull_Energy_Loop_01_distant}
 }
 {
  id=WavePlayer_18
  properties {bLooping=true, soundWaveAssetPtr=&/AB_Ambient_Loop_01_distant}
 }
 {id=WavePlayer_19, properties {soundWaveAssetPtr=&/AB_Pull_Loop_Sparks_01_distant}}
 {id=WavePlayer_2, properties {soundWaveAssetPtr=&/AB_Pull_Loop_Sparks_02}}
 {id=WavePlayer_20, properties {soundWaveAssetPtr=&/AB_Pull_Loop_Sparks_02_distant}}
 {id=WavePlayer_21, properties {soundWaveAssetPtr=&/AB_Pull_Loop_Sparks_03_distant}}
 {
  id=WavePlayer_22
  properties {bLooping=true, soundWaveAssetPtr=&/AB_Electrical_Arc_Loop_01_distant}
 }
 {id=WavePlayer_23, properties {soundWaveAssetPtr=&/AB_Pull_Loop_Sparks_04_distant}}
 {id=WavePlayer_3, properties {soundWaveAssetPtr=&/AB_Pull_Loop_Sparks_03}}
 {id=WavePlayer_4, properties {soundWaveAssetPtr=&/AB_Pull_Loop_Sparks_04}}
 {
  id=WavePlayer_5
  properties {bLooping=true, soundWaveAssetPtr=&/AB_Pull_Ice_Stress_Loop_01}
 }
 {id=WavePlayer_6, properties {bLooping=true, soundWaveAssetPtr=/Game/Sounds/Fort_GamePlay_Sounds/BRSword/AB_Pull_Energy_Loop_01}}
 {id=WavePlayer_7, properties {bLooping=true, soundWaveAssetPtr=&/AB_Electrical_Arc_Loop_01}}
]
properties {bOverrideConcurrency=true, maxDistance=25000, volumeMultiplier=0.9, concurrencyOverrides {resolutionRule=PreventNew}}
